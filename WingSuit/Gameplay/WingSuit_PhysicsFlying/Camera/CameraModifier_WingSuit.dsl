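; Wingsuit camera modifier that reshapes the player's view during flight and passes it through unchanged when the pawn has the CameraModes tag.
; When native MME is off, an optional gamepad analog-stick path trails the pawn with right-stick steering and clamped pitch.
; The other path layers an eased lead-in, velocity-driven offsets and an FOV boost, with an ADS alpha and modifier alpha scaling.
; WindDownAndRemove fades the modifier out, then removes it.
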(event Custom|WindDownAndRemove
  (CameraModifier|DisableModifier false)
  (Utilities|FlowControl|Delay (Variables|CameraModifier|GetAlphaOutTime))
  (Game|Player|RemoveCameraModifier (Variables|CameraModifier|GetCameraOwner) self))

(event EventBlueprintAddedToCamera
  (params (Camera :type Player%20Camera%20Manager%20Object%20Reference))
  (Variables|Default|SetLastFrameRotation (Camera|GetCameraRotation Camera))
  (Variables|Default|SetLastFrameRotation_RightStickInput "0, 0, 0")
  (Variables|Default|SetLastFramePosition (Transformation|GetActorLocation (Variables|CameraModifier|GetCameraOwner))))

(fn BlueprintModifyCamera
  (params (DeltaTime :type Float%20%28single-precision%29) (ViewLocation :type Vector :default "0, 0, 0") (ViewRotation :type Rotator :default "0, 0, 0") (FOV :type Float%20%28single-precision%29))
  (bind _velocity (LocalVariables|GetVelocity))
  (bind _asfort_player_controller (Utilities|Casting|CastToFortPlayerController (Game|Player|GetOwningPlayerController (Variables|CameraModifier|GetCameraOwner)))
    (:then
      (if (GameplayTags|HasMatchingGameplayTag (Game|Player|GetPlayerPawn _asfort_player_controller) "(TagName=\"CameraController.CameraModes\")")
        (return (Variables|GetViewLocation) (Variables|GetViewRotation) (Variables|GetFOV))
        (else
          _)))
    (:CastFailed
      _))
  ;[1 1 1] As of 9/15 we are using the nativized version as much as possible
  (if (not (ScalableFloat|GetValueasBool (Variables|Default|GetUseNativeMME)))
    (bind _asplayer_pawn_athena (Utilities|Casting|CastToPlayerPawn_Athena (Game|Player|GetPlayerPawn (Utilities|Casting|CastToFortPlayerController (Game|Player|GetOwningPlayerController (Variables|CameraModifier|GetCameraOwner))))))
    (Variables|Default|SetPlayerPawn_Athena _asplayer_pawn_athena)
    (bind _asfort_player_controller_2 (Utilities|Casting|CastToFortPlayerController (Pawn|GetController (Variables|Default|GetPlayerPawn_Athena))))
    (if (Input|IsUsingGamepad _asfort_player_controller_2)
      (if (ScalableFloat|GetValueasBool (Variables|Default|GetGamepadLeftStickAnalog))
        ;[1 1 1] If No Camera input is given - use the default camera for everyone else
        (bind _asfort_player_controller_3 (Utilities|Casting|CastToFortPlayerController (Game|Player|GetOwningPlayerController (Variables|CameraModifier|GetCameraOwner))))
        (LocalVariables|SetNoInputGiven (and (Math|Float|NearlyEqual(Float) (Game|Player|GetInputAnalogStickState _asfort_player_controller_3 "CAS_RightStick") 0.0 0.1) (Math|Float|NearlyEqual(Float) 0.0 (.StickY (Game|Player|GetInputAnalogStickState _asfort_player_controller_3 "CAS_RightStick")) 0.1)))
        ;[1 1 1] Find the Ideal Rotation
        (bind _asplayer_pawn_athena_2 (Utilities|Casting|CastToPlayerPawn_Athena (Game|Player|GetPlayerPawn (Utilities|Casting|CastToFortPlayerController (Game|Player|GetOwningPlayerController (Variables|CameraModifier|GetCameraOwner))))))
        (Variables|Default|SetPlayerPawn_Athena _asplayer_pawn_athena_2)
        (LocalVariables|SetNewRotation (Transformation|GetActorRotation (Variables|Default|GetPlayerPawn_Athena)))
        ;;[1 1 1] Rotate the camera based on input
        (bind _asfort_player_controller_4 (Utilities|Casting|CastToFortPlayerController (Game|Player|GetOwningPlayerController (Variables|CameraModifier|GetCameraOwner))))
        (LocalVariables|SetNoInputGiven (and (Math|Float|NearlyEqual(Float) (Game|Player|GetInputAnalogStickState _asfort_player_controller_4 "CAS_RightStick") 0.0 0.1) (Math|Float|NearlyEqual(Float) 0.0 (.StickY (Game|Player|GetInputAnalogStickState _asfort_player_controller_4 "CAS_RightStick")) 0.1)))
        ;;;[1 1 1] Testing add a negative pitch - this forces the player to look more down which keeps more of the world in frame
        (bind _output_get (LocalVariables|SetNewRotation (Math|Rotator|MakeRotator (.roll (LocalVariables|GetNewRotation)) (+ (.pitch (LocalVariables|GetNewRotation)) (Math|Float|MapRangeClamped (.StickY (Game|Player|GetInputAnalogStickState _asfort_player_controller_4 "CAS_RightStick")) 1.0 -1.0 -6e+01 6e+01)) (+ (Math|Float|MapRangeClamped (Game|Player|GetInputAnalogStickState _asfort_player_controller_4 "CAS_RightStick") 1.0 -1.0 9e+01 -9e+01) (.yaw (LocalVariables|GetNewRotation))))))
        (bind _output_get_2 (LocalVariables|SetNewRotation (Math|Rotator|MakeRotator (.roll _output_get) (+ (.pitch _output_get) (Variables|Default|GetRow_ArbitraryExtraDownwardPitch)) (.yaw _output_get))))
        ;[1 1 1] Clamp Max/Min Pitch; Never go past 90 degrees of the camera will rotate
        (LocalVariables|SetNewRotation (Math|Rotator|MakeRotator (.roll _output_get_2) (Math|Float|Clamp(Float) (.pitch _output_get_2) (ScalableFloat|GetValueasDouble (Variables|Default|GetMinPitch)) (ScalableFloat|GetValueasDouble (Variables|Default|GetMaxPitch_AnalogStickOnly))) (.yaw _output_get_2)))
        ;[1 1 1] Lerp towards ideal rotation
        (if (LocalVariables|GetNoInputGiven)
          (LocalVariables|SetCameraFollowMulti (ScalableFloat|GetValueasDouble (Variables|Default|GetCameraFollowMultiplier_Rotation)))
          (else
            (LocalVariables|SetCameraFollowMulti (ScalableFloat|GetValueasDouble (Variables|Default|GetCameraFollowMultiplier_RotationInputGiven)))))
        (bind _returnvalue (Fortnite|RotateVectorTowardVector (Math|Vector|GetForwardVector (Variables|Default|GetLastFrameRotation)) (Math|Vector|GetForwardVector (LocalVariables|GetNewRotation)) (* (Utilities|Time|GetWorldDeltaSeconds) (LocalVariables|GetCameraFollowMulti))))
        (LocalVariables|SetNewRotation (Math|Rotator|MakeRotfromXZ _returnvalue (Math|Vector|MakeVector 0.0 0.0 1.0)))
        (LocalVariables|SetNewCameraLocation (+ (* (Math|Vector|GetForwardVector (LocalVariables|GetNewRotation)) (Math|Float|MapRangeClamped (Math|Vector|VectorLength (Transformation|GetVelocity (Variables|Default|GetPlayerPawn_Athena))) 0.0 4.5e+03 -6e+02 -1.6e+03)) (Transformation|GetActorLocation (Variables|Default|GetPlayerPawn_Athena))))
        ;[1 1 1] Lerp towards Ideal Position
        (Variables|Default|SetLastFramePosition (Math|Vector|Lerp(Vector) (Variables|Default|GetLastFramePosition) (LocalVariables|GetNewCameraLocation) (* (Utilities|Time|GetWorldDeltaSeconds) (Variables|Default|GetCameraFollowMultiplier_Position))))
        (bind _output_get_3 (Variables|Default|SetLastFrameRotation (LocalVariables|GetNewRotation)))
        (return (LocalVariables|GetNewCameraLocation) _output_get_3 (Variables|GetFOV))
        (else
          _))))
  ;[1 0.9792 0.9792] No need for engineering review here - this camera should be nativized before shipping
  (Utilities|FlowControl|Sequence
    (:then_0
      (Utilities|IsValid (Variables|Default|GetPlayerPawn_Athena)
        (:Is%20Not%20Valid
          (bind _asplayer_pawn_athena_3 (Utilities|Casting|CastToPlayerPawn_Athena (Game|Player|GetPlayerPawn (Utilities|Casting|CastToFortPlayerController (Game|Player|GetOwningPlayerController (Variables|CameraModifier|GetCameraOwner))))))
          (Variables|Default|SetPlayerPawn_Athena _asplayer_pawn_athena_3)
          (Variables|Default|SetAddedTime (Utilities|Time|GetGameTimeinSeconds))
          (Variables|Default|SetAnchoredLocation (Transformation|GetActorLocation (Variables|Default|GetPlayerPawn_Athena))))))
    (:then_1
      (if (< (- (Utilities|Time|GetGameTimeinSeconds) (Variables|Default|GetAddedTime)) (Variables|Default|GetLeadInTime))
        (Variables|Default|SetLeadIn_ForwardOffset (* (Math|Vector|DotProduct (Math|Vector|GetForwardVector (Variables|GetViewRotation)) (- (Transformation|GetActorLocation (Variables|Default|GetPlayerPawn_Athena)) (Variables|Default|GetAnchoredLocation))) (Utilities|UnNormalizeFloat (Math|Curves|GetFloatValue (Variables|Default|GetForwardPercentCurve) (Utilities|NormalizeFloat (- (Utilities|Time|GetGameTimeinSeconds) (Variables|Default|GetAddedTime)) 0 (Variables|Default|GetLeadInTime) true)) 0 0.8)))))
    (:then_2
      (bind _output_get_4 (LocalVariables|SetVelocity (/ (- (Transformation|GetActorLocation (Variables|Default|GetPlayerPawn_Athena)) (Variables|Default|GetLastFramePosition_)) (Math|Float|SelectFloat 1.0 (Variables|GetDeltaTime) (Math|Float|NearlyEqual(Float) (Variables|GetDeltaTime))))))
      (LocalVariables|SetDirection (Math|Vector|Normalize _output_get_4))
      (LocalVariables|SetSpeed (Math|Vector|VectorLength _output_get_4))
      (Variables|Default|SetLastFramePosition_ (Transformation|GetActorLocation (Variables|Default|GetPlayerPawn_Athena))))
    (:then_3
      (Variables|Default|SetADS_Alpha (Math|Interpolation|FInterpTo (Variables|Default|GetADS_Alpha) (Math|Float|SelectFloat 1.0 0.0 (LocalVariables|GetIsShooting)) (Variables|GetDeltaTime) 4.0))
      (LocalVariables|SetForwardSpeed (Math|Vector|DotProduct _velocity (Math|Vector|GetForwardVector (Variables|GetViewRotation))))
      (LocalVariables|SetRightSpeed (Math|Vector|DotProduct _velocity (Math|Vector|GetRightVector (Variables|GetViewRotation))))
      (LocalVariables|SetUpSpeed (Math|Vector|DotProduct _velocity (Math|Vector|GetUpVector (Variables|GetViewRotation))))
      (LocalVariables|SetCurrentStraightness (Math|Float|Absolute(Float) (Math|Vector|DotProduct (Math|Vector|GetForwardVector (Variables|GetViewRotation)) (LocalVariables|GetDirection))))
      (Variables|Default|SetCurrentForward (Math|Interpolation|FInterpTo (Variables|Default|GetCurrentForward) (Utilities|Operators|Add (* (Utilities|RemapFloat (Math|Vector|VectorLengthXY (LocalVariables|GetVelocity)) "(X=800.000000,Y=4000.000000)" "(X=45.000000,Y=75.000000)" true) (Math|Float|SelectFloat 3.0 1.0 (GameplayTags|HasMatchingGameplayTag (Variables|Default|GetPlayerPawn_Athena) "(TagName=\"WingSuit.SmokeScreening\")"))) 0 (Utilities|RemapFloat (Math|Float|Absolute(Float) (LocalVariables|GetUpSpeed)) "(X=0.000000,Y=300.000000)" "(X=0.000000,Y=42.000000)" true) (Math|Float|SelectFloat 1e+01 0.0 (Variables|/Script/FortniteGame.FortPlayerPawn|GetbIsTacticalSprinting (Variables|Default|GetPlayerPawn_Athena)))) (Variables|GetDeltaTime) 3.0))
      (Variables|Default|SetCurrentRight (Math|Interpolation|FInterpTo (Variables|Default|GetCurrentRight) (Utilities|Operators|Multiply (Utilities|RemapFloat (LocalVariables|GetRightSpeed) "(X=-500.000000,Y=500.000000)" "(X=-150.000000,Y=150.000000)" true) (Math|Float|SelectFloat 1.1 0.9 (> (Utilities|RemapFloat (LocalVariables|GetRightSpeed) "(X=-500.000000,Y=500.000000)" "(X=-150.000000,Y=150.000000)" true) 0)) 1.0 (Utilities|RemapFloat (Math|Float|Absolute(Float) (Variables|Default|GetCurrentForward)) "(X=50.000000,Y=300.000000)" "(X=1.200000,Y=1.500000)" true)) (Variables|GetDeltaTime) 3.0))
      (Variables|Default|SetCurrentUp (Math|Interpolation|FInterpTo (Variables|Default|GetCurrentUp) (+ (Utilities|Operators|Multiply (Utilities|RemapFloat (LocalVariables|GetUpSpeed) "(X=-1000.000000,Y=1000.000000)" "(X=-170.000000,Y=170.000000)" true) (Utilities|RemapFloat (Math|Float|Absolute(Float) (Variables|Default|GetCurrentForward)) "(X=50.000000,Y=400.000000)" "(X=1.000000,Y=1.600000)" true) (Math|Float|SelectFloat (ScalableFloat|GetValueatLevel (Variables|Default|GetSprintingUpOffsetMultiplier)) (ScalableFloat|GetValueatLevel (Variables|Default|GetBaseUpOffsetMultiplier)) (Variables|/Script/FortniteGame.FortPlayerPawn|GetbIsTacticalSprinting (Variables|Default|GetPlayerPawn_Athena))) (Math|Float|SelectFloat 1.35 1.0 (> (Utilities|RemapFloat (LocalVariables|GetUpSpeed) "(X=-1000.000000,Y=1000.000000)" "(X=-170.000000,Y=170.000000)" true) 0))) (Math|Float|SelectFloat -7e+01 -4e+01 (Variables|/Script/FortniteGame.FortPlayerPawn|GetbIsTacticalSprinting (Variables|Default|GetPlayerPawn_Athena)))) (Variables|GetDeltaTime) 4.0))
      (Variables|Default|SetCurrentLean (Math|Interpolation|FInterpTo (Variables|Default|GetCurrentLean) (Utilities|RemapFloat (* (LocalVariables|GetCurrentStraightness) (LocalVariables|GetLeanRight)) "(X=-1.000000,Y=1.000000)" (Math|Vector2D|MakeVector2D (Utilities|RemapFloat (Math|Float|Absolute(Float) (LocalVariables|GetForwardSpeed)) "(X=800.000000,Y=1800.000000)" "(X=40.000000,Y=50.000000)" true) (* (Utilities|RemapFloat (Math|Float|Absolute(Float) (LocalVariables|GetForwardSpeed)) "(X=800.000000,Y=1800.000000)" "(X=40.000000,Y=50.000000)" true) -2.75)) true) (Variables|GetDeltaTime) 4.5))
      (Variables|Default|SetCurrentFOV (Math|Interpolation|FInterpTo (Variables|Default|GetCurrentFOV) (Math|Float|SelectFloat 17.0 23.0 (Variables|/Script/FortniteGame.FortPlayerPawn|GetbIsTacticalSprinting (Variables|Default|GetPlayerPawn_Athena))) (Variables|GetDeltaTime) 2.25))
      (Variables|Default|SetCurrentOffset (Math|Interpolation|FInterpTo (Variables|Default|GetCurrentForward) (Utilities|RemapFloat (LocalVariables|GetSpeed) "(X=-3500.000000,Y=3500.000000)" "(X=-220.000000,Y=220.000000)" true) (Variables|GetDeltaTime) 4.0))
      (Variables|Default|SetCurrentOffset_Vec (Math|Interpolation|VInterpTo (Variables|Default|GetCurrentOffset_Vec) (Utilities|Operators|Add (* (Math|Vector|GetForwardVector (Variables|GetViewRotation)) (Variables|Default|GetCurrentForward)) (* (Math|Vector|GetRightVector (Variables|GetViewRotation)) (+ (Variables|Default|GetCurrentRight) (* (LocalVariables|GetCurrentStraightness) -9e+01))) (* "0.000000,0.000000,1.000000" (- (Variables|Default|GetCurrentUp) 18.0)) (* (Math|Vector|CrossProduct (LocalVariables|GetDirection) "0.000000,0.000000,1.000000") (Variables|Default|GetCurrentLean))) (Variables|GetDeltaTime) 1e+01)))
    (:then_4
      (return (- (Variables|GetViewLocation) (+ (* (Utilities|Operators|Add (* (Variables|Default|GetCurrentOffset_Vec) (- 1.0 (Variables|Default|GetADS_Alpha))) (Utilities|Operators|Multiply (Math|Vector|GetForwardVector (Variables|GetViewRotation)) (- 1.0 (Variables|Default|GetADS_Alpha)) -3e+01) (* (Math|Vector|GetUpVector (Variables|GetViewRotation)) 3e+01) (* (Math|Vector|GetRightVector (Variables|GetViewRotation)) -1e+01)) (Utilities|CalculateAlpha)) (* (Math|Vector|GetForwardVector (Variables|GetViewRotation)) (Variables|Default|GetLeadIn_ForwardOffset)))) (Variables|GetViewRotation) (+ (Variables|GetFOV) (Utilities|Operators|Multiply (Variables|Default|GetCurrentFOV) (Utilities|CalculateAlpha) (- 1.0 (Variables|Default|GetADS_Alpha))))))))

(fn SetStartingLocation
  (params (StartingLocation :type Vector))
  (Variables|Default|SetStartingLocation StartingLocation))
