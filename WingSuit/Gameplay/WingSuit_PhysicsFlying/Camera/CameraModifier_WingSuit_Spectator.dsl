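; Camera modifier for spectating wingsuit flight, offsetting the view along its forward vector by CurrentForward scaled by alpha.
; When enabled, a hotfix-gated looping timer checks that the view target is a FortPlayerPawn tagged as wingsuit flying.
; Otherwise it disables itself, waits out the alpha blend, then removes itself from the camera manager.
; Last-frame values, starting location and post-boost flag are stored but never read here; SetStartingValues is empty.
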(event Custom|WindDownAndRemove
  (CameraModifier|DisableModifier false)
  (Utilities|FlowControl|Delay (Variables|CameraModifier|GetAlphaOutTime))
  (Game|Player|RemoveCameraModifier (Variables|CameraModifier|GetCameraOwner) self))

(event EventBlueprintAddedToCamera
  (params (Camera :type Player%20Camera%20Manager%20Object%20Reference))
  (Variables|Default|SetLastFrameRotation (Camera|GetCameraRotation Camera))
  (Variables|Default|SetLastFrameRotation_RightStickInput "0, 0, 0")
  (bind _asfort_player_controller (Utilities|Casting|CastToFortPlayerController (Game|Player|GetOwningPlayerController (Variables|CameraModifier|GetCameraOwner))))
  (Variables|Default|SetLastFramePosition_ (Transformation|GetActorLocation (Pawn|GetViewTarget _asfort_player_controller))))

(event Custom|RemovePostBoostMultilpier
  (Variables|Default|SetPostBoostingMultiplier false))

(event Custom|AddPostBoostMultiplier
  (Variables|Default|SetPostBoostingMultiplier true))

(event EventBlueprintModifierEnabled
  (if (ScalableFloat|GetValueasBool (Variables|Default|GetCheckViewTargetTickHotFix))
    (bind _returnvalue (Utilities|Time|SetTimerbyEvent (delegate Custom|CheckViewTarget) 0.25 true true))
    (Variables|Default|SetCheckViewTargetTickHandle _returnvalue)))

(event Custom|CheckViewTarget
  (bind _asfort_player_pawn (Utilities|Casting|CastToFortPlayerPawn (Pawn|GetViewTarget (Game|Player|GetOwningPlayerController (Variables|CameraModifier|GetCameraOwner))))
    (:then
      (if (not (GameplayTags|HasMatchingGameplayTag _asfort_player_pawn "(TagName=\"MovementMode.Flying.WingSuit\")"))
        (CallFunction|WindDownandRemove)))
    (:CastFailed
      (CallFunction|WindDownandRemove))))

(event EventBlueprintModifierDisabled
  (params (bImmediateDisable :type Boolean))
  (Utilities|Time|ClearandInvalidateTimerbyHandle (Variables|Default|GetCheckViewTargetTickHandle)))

(fn BlueprintModifyCamera
  (params (DeltaTime :type Float%20%28single-precision%29) (ViewLocation :type Vector :default "0, 0, 0") (ViewRotation :type Rotator :default "0, 0, 0") (FOV :type Float%20%28single-precision%29))
  (return (+ (* (Math|Vector|GetForwardVector (Variables|GetViewRotation)) (* (Variables|Default|GetCurrentForward) (Variables|CameraModifier|GetAlpha))) (Variables|GetViewLocation)) (Variables|GetViewRotation) (Variables|GetFOV)))

(fn SetStartingLocation
  (params (StartingLocation :type Vector))
  (Variables|Default|SetStartingLocation StartingLocation))

(fn SetStartingValues
  (params (Camera :type Player%20Camera%20Manager%20Object%20Reference)))
